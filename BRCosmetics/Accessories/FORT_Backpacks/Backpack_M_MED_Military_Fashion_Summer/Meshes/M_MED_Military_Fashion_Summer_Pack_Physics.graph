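# Physics asset for the Military Fashion Summer backpack cosmetic, defining collision bodies and joints for its dangling parts.
# A large box body covers the pack, with capsules for the pouch and stick and boxes for the two zipper pieces.
# Pouch, stick and first zipper constrain to the pack, while the second zipper chains off the first.
# Tight angular limits keep pouch and stick nearly rigid, while the zippers swing more freely.
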PhysicsAsset /BRCosmetics/Accessories/FORT_Backpacks/Backpack_M_MED_Military_Fashion_Summer/Meshes/M_MED_Military_Fashion_Summer_Pack_Physics
bodies {
 dyn_pouch [{shapeType=Capsule, length=7, radius=2.5, center=<1.7872 0 0>, rotation=<rot 90 0 0>}]
 dyn_stick [{shapeType=Capsule, length=21.9759, radius=1.9338, center=<1.8737 0.0152 -0.2843>, rotation=<rot 0 90.0154 -89.9314>}]
 dyn_zipper_a [{shapeType=Box, extentX=0.5, extentY=1.2004, extentZ=2.7846, center=<1.0662 -0.0362 0.4155>, rotation=<rot 0 89.1709 -88.1944>}]
 dyn_zipper_b [{shapeType=Box, extentX=2.5, extentY=1.5, extentZ=1.5, center=<1.4342 0 0>}]
 pack [{shapeType=Box, extentX=5.9462, extentY=18.8052, extentZ=34.37, center=<3.3871 -7.0183 -3.6119>, rotation=<rot 10.0257 -86.1487 26.8256>}]
}
constraints [
 {bone1Name=dyn_stick, bone2Name=pack, swing1LimitDegrees=3, swing2LimitDegrees=3, twistLimitDegrees=1}
 {bone1Name=dyn_zipper_a, bone2Name=pack, swing1LimitDegrees=30, swing2LimitDegrees=30, twistLimitDegrees=15}
 {bone1Name=dyn_zipper_b, bone2Name=dyn_zipper_a, swing1LimitDegrees=30, swing2LimitDegrees=30, twistLimitDegrees=15}
 {bone1Name=dyn_pouch, bone2Name=pack, swing1LimitDegrees=5, swing2LimitDegrees=5, twistLimitDegrees=5}
]
